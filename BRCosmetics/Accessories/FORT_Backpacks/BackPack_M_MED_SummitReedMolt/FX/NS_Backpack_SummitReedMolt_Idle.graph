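# Niagara idle effect for the SummitReedMolt backpack cosmetic, drawing audio-reactive rings and smoke at backpack sockets.
# RingsTop and RingsMiddle burst mesh and sprite rings gated by an audio threshold, with separate audio channels driving size and thickness.
# Smoke holds two mirrored, audio-tinted meshes with interpolated position data, while SmokeDetailed adds drifting flipbook sprites.
# System-level audio values, including a fake-loop module, feed emitters alongside front-end and time-of-day tuning.
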NiagaraSystem /BRCosmetics/Accessories/FORT_Backpacks/BackPack_M_MED_SummitReedMolt/FX/NS_Backpack_SummitReedMolt_Idle
dependencies {
 usedDataInterfaces [
  &NiagaraDataInterfaceCurve
  &NiagaraDataInterfaceArrayInt32
  &NiagaraDataInterfaceVectorCurve
  &NiagaraDataInterfaceVector2DCurve
  &NiagaraDataInterfaceSpriteRendererInfo
  &NiagaraDataInterfaceSocketReader
  &NiagaraDataInterfaceColorCurve
 ]
 usedDynamicInputs [
  &Vector2DFromCurve
  &RandomRangeVector2D
  &MakeCustomFloatFromBool
  &VectorLength
  &ClampFloat
  &Waveform
  &MakeFloatFromLinearColor
 ]
 usedModules [
  &self:SystemSpawnScript.SystemScriptSource.SystemScriptGraph.NiagaraNodeAssignment_1.SetVariables_#1
  &self:SystemSpawnScript.SystemScriptSource.SystemScriptGraph.NiagaraNodeAssignment_4.SetVariables_#2
  &SystemState
  &self:SystemSpawnScript.SystemScriptSource.SystemScriptGraph.NiagaraNodeAssignment_2.SetVariables_#3
  /Game/Effects/Niagara/Modules/Audio/FN_FakeAudioReactivityLoop
  &self:SystemSpawnScript.SystemScriptSource.SystemScriptGraph.NiagaraNodeAssignment_5.SetVariables_#4
  &self:SystemSpawnScript.SystemScriptSource.SystemScriptGraph.NiagaraNodeAssignment_6.SetVariables_#5
  &self:SystemSpawnScript.SystemScriptSource.SystemScriptGraph.NiagaraNodeAssignment_7.SetVariables_#6
  &self:SystemSpawnScript.SystemScriptSource.SystemScriptGraph.NiagaraNodeAssignment_8.SetVariables_#7
  &Timeline
 ]
}
emitters {
 RingsMiddle {
  emitterUpdateScript [
   &EmitterState()
   &SpawnBurst_Instantaneous(
    Spawn Count=&MakeCustomIntFromBool(
     Bool=&SetBoolByFloatComparison(A=$System.IncomingAudio, B=0.1, Comparison Type=::NewEnumerator2)
     True INT=2
     False INT=0
    )
    Use Spawn Probability=false
    UseLoopCountLimit=false
   ) {module=SpawnBurst_Instantaneous001}
   &SpawnBurst_Instantaneous(
    Spawn Count=&MakeCustomIntFromBool(
     Bool=&SetBoolByFloatComparison(A=$System.IncomingAudio, B=0.1, Comparison Type=::NewEnumerator2)
     True INT=2
     False INT=0
    )
    Spawn Time=0.3
    Use Spawn Probability=false
    UseLoopCountLimit=false
   ) {module=SpawnBurst_Instantaneous002}
   &SpawnBurst_Instantaneous(
    Spawn Count=&MakeCustomIntFromBool(
     Bool=&SetBoolByFloatComparison(A=$System.IncomingAudio, B=0.1, Comparison Type=::NewEnumerator2)
     True INT=2
     False INT=0
    )
    Spawn Time=0.15
    Use Spawn Probability=false
    UseLoopCountLimit=false
   ) {module=SpawnBurst_Instantaneous003}
  ]
  particleSpawnScript [
   &PartitionParticles()
   &InitializeParticle(
    Color Mode=::NewEnumerator0
    Lifetime Max=0.8
    Lifetime Min=0.7
    Lifetime Mode=::NewEnumerator1
    Mesh Scale Mode=::NewEnumerator3
    Mesh Uniform Scale=0.8
    Sprite Size Mode=::NewEnumerator3
    Uniform Sprite Size=19.588
    UsePositionOffset=false
    Write Material Random=true
   )
   &InitialMeshOrientation(Mesh Axis To Orient=<0 0 1>, Mesh Orientation Mode=::NewEnumerator2, Use Rotation Vector=false)
   &CameraOffset(Camera Offset Amount=5)
  ]
  particleUpdateScript [
   &ParticleState()
   &ScaleMeshSize(Scale Factor=&VectorFromFloat(Value=&Multiply_Float(
      A=&FloatFromCurve(FloatCurve=DataInterface() {propertyValues {Curve {
          defaultValue=1
          keys [
           {interpMode=RCIM_Cubic, tangentMode=RCTM_User, time=0, value=0.3}
           {arriveTangent=1.7724, interpMode=RCIM_Cubic, leaveTangent=1.7724, tangentMode=RCTM_User, time=1, value=1}
          ]
         }}})
      B=&Lerp_Float(A=0.75, B=1.5, Alpha=$System.IncomingAudio_G)
     ))) {module=ScaleMeshSize001}
   &SocketLocation(
    SamplePosOffset=true
    Sampled Position Offset=&MakeVector(
     X=&Add_Float(
      A=&FloatFromCurve(
       FloatCurve=DataInterface() {propertyValues {Curve {
          defaultValue=1
          keys [
           {arriveTangent=-0.0014, interpMode=RCIM_Cubic, leaveTangent=-0.0014, tangentMode=RCTM_User, time=0, value=0}
           {arriveTangent=3.0805, interpMode=RCIM_Cubic, leaveTangent=3.0805, tangentMode=RCTM_User, time=1, value=1}
          ]
         }}}
       Scale Curve=2
      )
      B=0.7
     )
     Y=0
     Z=0
    )
    Socket Index=&SelectIntFromArray(Direct Array Index=$Particles.PartitionParticles.Partition, Int Selection Array=DataInterface() {propertyValues {IntData [4,2]}})
    Socket Reader=$System.Socket Reader
    Socket Sampling Mode=::NewEnumerator3
    VelocitySpeedLimit=false
   )
   &DynamicMaterialParameters(
    Index 0 Param 1=&Lerp_Float(A=0.01, B=$System.Rings_ThicknessScale, Alpha=$System.IncomingAudio_G)
    Index 0 Param 4=0.01
    Module.Index0Float0Write=true
    Module.Index0Float1Write=true
    Module.Index0Float2Write=true
    Module.Index0Float3Write=false
   ) {module=DynamicMaterialParameters001}
   &ScaleColor(
    Scale Alpha=&Multiply_Float(
     A=&FloatFromCurve(
      FloatCurve=DataInterface() {propertyValues {Curve {
         defaultValue=1
         keys [{interpMode=RCIM_Cubic, tangentMode=RCTM_User, time=0.7, value=1}, {interpMode=RCIM_Cubic, tangentMode=RCTM_User, time=1, value=0}]
        }}}
      Scale Curve=&FNTimeOfDayScalar(Night Scale=0.6, Morning Scale=0.9, Evening Scale=0.8)
     )
     B=$System.Timeline001.Time
    )
    Scale RGB=&MakeVectorFromLinearColorRGB(Color=$System.Rings_ScaleColor)
    ScaleA=true
    ScaleRGB=true
   )
   /SetVariables(Particles.GlowColor=&Multiply_LinearColorByFloat(
     Linear Color=&Multiply_LinearColorByFloat(Linear Color=$Particles.Color, Float=1.25)
     Float=&FN_FrontEndFloatModifier(Gameplay=0.51, Front End=0.33)
    )) {isSetParameters=true}
  ]
  rendererClasses [&NiagaraMeshRendererProperties, &NiagaraSpriteRendererProperties]
  renderers [
   {rendererClass=&NiagaraMeshRendererProperties, rendererIndex=0}
   {rendererClass=&NiagaraSpriteRendererProperties, rendererIndex=1}
  ]
 }
 RingsTop {
  emitterUpdateScript [
   &EmitterState()
   &SpawnBurst_Instantaneous(
    Spawn Count=&MakeCustomIntFromBool(
     Bool=&SetBoolByFloatComparison(A=$System.IncomingAudio, B=0.1, Comparison Type=::NewEnumerator2)
     True INT=2
     False INT=0
    )
    Spawn Time=0.5
    Use Spawn Probability=false
    UseLoopCountLimit=false
   ) {module=SpawnBurst_Instantaneous001}
   &SpawnBurst_Instantaneous(
    Spawn Count=&MakeCustomIntFromBool(
     Bool=&SetBoolByFloatComparison(A=$System.IncomingAudio, B=0.1, Comparison Type=::NewEnumerator2)
     True INT=2
     False INT=0
    )
    Spawn Time=0.2
    Use Spawn Probability=false
    UseLoopCountLimit=false
   ) {module=SpawnBurst_Instantaneous002}
  ]
  particleSpawnScript [
   &PartitionParticles()
   &InitializeParticle(
    Color Mode=::NewEnumerator0
    Lifetime Max=0.8
    Lifetime Min=0.7
    Lifetime Mode=::NewEnumerator1
    Mesh Scale Mode=::NewEnumerator3
    Mesh Uniform Scale=0.35
    Sprite Size Mode=::NewEnumerator3
    Uniform Sprite Size=16
    UsePositionOffset=false
    Write Material Random=true
   )
   &InitialMeshOrientation(Mesh Axis To Orient=<0 0 1>, Mesh Orientation Mode=::NewEnumerator2, Use Rotation Vector=false)
   &CameraOffset(Camera Offset Amount=5)
  ]
  particleUpdateScript [
   &ParticleState()
   &ScaleMeshSize(Scale Factor=&VectorFromFloat(Value=&Multiply_Float(
      A=&FloatFromCurve(FloatCurve=DataInterface() {propertyValues {Curve {
          defaultValue=1
          keys [
           {interpMode=RCIM_Cubic, tangentMode=RCTM_User, time=0, value=0.3}
           {arriveTangent=1.7724, interpMode=RCIM_Cubic, leaveTangent=1.7724, tangentMode=RCTM_User, time=1, value=1}
          ]
         }}})
      B=&Lerp_Float(A=0.75, B=1.5, Alpha=$System.IncomingAudio_R)
     ))) {module=ScaleMeshSize001}
   &SocketLocation(
    SamplePosOffset=true
    Sampled Position Offset=&MakeVector(
     X=&Add_Float(
      A=&FloatFromCurve(FloatCurve=DataInterface() {propertyValues {Curve {
          defaultValue=1
          keys [
           {arriveTangent=-0.0014, interpMode=RCIM_Cubic, leaveTangent=-0.0014, tangentMode=RCTM_User, time=0, value=0}
           {arriveTangent=3.0805, interpMode=RCIM_Cubic, leaveTangent=3.0805, tangentMode=RCTM_User, time=1, value=1}
          ]
         }}})
      B=0.5
     )
     Y=0
     Z=0
    )
    Socket Index=&SelectIntFromArray(Direct Array Index=$Particles.PartitionParticles.Partition, Int Selection Array=DataInterface() {propertyValues {IntData [3,1]}})
    Socket Reader=$System.Socket Reader
    Socket Sampling Mode=::NewEnumerator3
    VelocitySpeedLimit=false
   )
   &DynamicMaterialParameters(
    Index 0 Param 1=&Lerp_Float(A=0.01, B=$System.Rings_ThicknessScale, Alpha=$System.IncomingAudio_R)
    Index 0 Param 4=0.01
    Module.Index0Float0Write=true
    Module.Index0Float1Write=true
    Module.Index0Float2Write=true
    Module.Index0Float3Write=false
   ) {module=DynamicMaterialParameters001}
   &ScaleColor(
    Scale Alpha=&Multiply_Float(
     A=&FloatFromCurve(FloatCurve=DataInterface() {propertyValues {Curve {
         defaultValue=1
         keys [{interpMode=RCIM_Cubic, tangentMode=RCTM_User, time=0.7, value=1}, {interpMode=RCIM_Cubic, tangentMode=RCTM_User, time=1, value=0}]
        }}})
     B=$System.Timeline001.Time
    )
    Scale RGB=&MakeVectorFromLinearColorRGB(Color=$System.Rings_ScaleColor)
    ScaleA=true
    ScaleRGB=true
   )
   /SetVariables(Particles.GlowColor=&Multiply_LinearColorByFloat(
     Linear Color=&Multiply_LinearColorByFloat(Linear Color=$Particles.Color, Float=1.3)
     Float=&FN_FrontEndFloatModifier(Gameplay=0.51, Front End=0.38)
    )) {isSetParameters=true}
  ]
  rendererClasses [&NiagaraMeshRendererProperties, &NiagaraSpriteRendererProperties]
  renderers [
   {rendererClass=&NiagaraMeshRendererProperties, rendererIndex=0}
   {rendererClass=&NiagaraSpriteRendererProperties, rendererIndex=1}
  ]
 }
 Smoke {
  emitterUpdateScript [
   &EmitterState(
    Inactive Response=::NewEnumerator1
    Life Cycle Mode=::NewEnumerator1
    Loop Behavior=::NewEnumerator1
    Loop Duration Mode=::NewEnumerator1
    UseLoopDelay=false
   )
   &SpawnBurst_Instantaneous(Spawn Count=2, Use Spawn Probability=false, UseLoopCountLimit=false)
  ]
  particleSpawnScript [
   &FN_SetSpawnGroup(Number of Spawn Groups=2)
   &InitializeParticle(
    Mesh Scale=&MergeHelper_VectorSwitch(0=<1 1 0.965>, 1=<-1 1 0.965>, 2=?, 3=?, Index=&SpawnGroup(), Num=::NewEnumerator0)
    Mesh Scale Mode=::NewEnumerator4
    UsePositionOffset=false
    Write Material Random=true
   )
  ]
  particleUpdateScript [
   &ParticleState(Kill Particles When Lifetime Has Elapsed=false, Let Infinitely Lived Particles Die When Emitter Deactivates=true)
   &SocketLocation(
    SamplePosOffset=true
    Sampled Position Offset=&MergeHelper_VectorSwitch(0=<2 0.7 -0.5>, 1=<-2 0.7 -0.5>, 2=?, 3=?, Index=&SpawnGroup(), Num=::NewEnumerator0)
    Socket Reader=$System.Socket Reader
    Socket Sampling Mode=::NewEnumerator3
    VelocitySpeedLimit=false
   )
   &InterpolateOverTime(
    Rate Of Change:NiagaraFloat=&Multiply_Float(A=$System.Interpolate Rate, B=1)
    Target Vector Value:NiagaraPosition=$Particles.Position
    Variable Type=::NewEnumerator2
   )
   /SetVariables(Particles.Interpolate DY:Vector4f=&MakeFloat4FromVectorAndFloat("Vector (RGB)"=$Particles.InterpolateOverTime.InterpolateOverTime_V3_.Moving Average, "Float (Alpha)"=1)) {isSetParameters=true}
   &DynamicMaterialParameters(
    Index 0 Param 1=&Lerp_Float(A=1.4, B=1.7, Alpha=$System.IncomingAudio)
    Index 0 Param 2=1.45
    Index 0 Param 3=&Lerp_Float(A=2.3, B=4.5, Alpha=$System.IncomingAudio)
    Index 0 Param 4=&MergeHelper_FloatSwitch(0=0, 1=0.4, 2=?, 3=?, Index=&SpawnGroup(), Num=::NewEnumerator0)
    Index 2 Param 1=&Lerp_Float(A=2.2, B=4, Alpha=$System.IncomingAudio)
    Index 2 Param 2=&FN_FrontEndFloatModifier(Gameplay=2.5, Front End=4.3)
    Index 2 Param 3=0.62
    Index 2 Param 4=&Lerp_Float(A=0, B=0, Alpha=$System.IncomingAudio)
    Index 3 Linear Color:LinearColor=&Multiply_LinearColorByFloat(
     Linear Color=&LerpLinearColors(
      StartColor=<rgba 1 1 1 1>
      EndColor=&LerpLinearColors(StartColor=<rgba 0.3281 0.012 0.0145 1>, EndColor=<rgba 5 0.1823 0.2209 1>, LerpFactor=$System.IncomingAudio_B)
      LerpFactor=$System.IncomingAudio
     )
     Float=&FN_FrontEndFloatModifier(
      Gameplay=&FNTimeOfDayScalar(Night Scale=0.4, Morning Scale=0.9, Evening Scale=0.8)
      Front End=0.6
     )
    )
    Module.Index0Float0Write=true
    Module.Index0Float1Write=true
    Module.Index0Float2Write=true
    Module.Index0Float3Write=true
    Module.Index2Float0Write=true
    Module.Index2Float1Write=true
    Module.Index2Float2Write=true
    Module.Index2Float3Write=true
    Write Parameter Index 2=::NewEnumerator1
    Write Parameter Index 3=::NewEnumerator5
   )
   &ScaleColor(
    Scale Alpha=&FN_FrontEndFloatModifier(
     Gameplay=&Lerp_Float(A=26, B=22, Alpha=$System.IncomingAudio)
     Front End=&FloatFromCurve(
      FloatCurve=DataInterface() {propertyValues {Curve {keys [<at 0.5 0>,<at 1.2 1>]}}}
      CurveIndex=$Engine.System.Age
      Scale Curve=&Lerp_Float(A=18, B=16, Alpha=$System.IncomingAudio)
     )
    )
    Scale RGB=&MakeVectorFromLinearColorRGB(Color=&Multiply_LinearColorByFloat(
      Linear Color=&LerpLinearColors(StartColor=<rgba 0.016 0.016 0.016 1>, EndColor=<rgba 0.0573 0 0 1>, LerpFactor=$System.IncomingAudio)
      Float=&FNTimeOfDayScalar(Night Scale=0.6, Morning Scale=0.9, Evening Scale=0.8)
     ))
    ScaleA=true
    ScaleRGB=true
   )
   &ScaleMeshSize(Scale Factor=$Output.SocketLocation.SampledScale)
   &CameraOffset(Camera Offset Amount=&FloatFromCurve(
     FloatCurve=DataInterface() {propertyValues {Curve {keys [<at 0 2 RCIM_Constant>,<at 0.1211 2 RCIM_Constant>,<at 0.2025 3.2 RCIM_Constant>]}}}
     CurveIndex=&RemapRange(
      Input Value To Remap=&MakeFloatFromVector(Vector=$Output.SocketLocation.SampledScale, Channel=::NewEnumerator0)
      Input Min=0.93
      Input Max=1.2
      Output Min=0
      Output Max=1
     )
    ))
  ]
  rendererClasses [&NiagaraMeshRendererProperties]
  renderers [{rendererClass=&NiagaraMeshRendererProperties, rendererIndex=0}]
 }
 SmokeDetailed {
  emitterUpdateScript [
   &EmitterState(Inactive Response=::NewEnumerator1, Life Cycle Mode=::NewEnumerator1, Loop Delay=0.4, Loop Duration=2, UseLoopDelay=false)
   &SpawnRate(SpawnRate=10, Use Spawn Probability=false)
  ]
  particleSpawnScript [
   &FN_SetSpawnGroup(Number of Spawn Groups=2)
   &InitializeParticle(
    Lifetime Max=1.6
    Lifetime Min=1.4
    Lifetime Mode=::NewEnumerator1
    Sprite Rotation Mode=::NewEnumerator2
    Sprite Size Max=<100 100>
    Sprite Size Min=<80 80>
    Sprite Size Mode=::NewEnumerator2
    UsePositionOffset=true
    Write Material Random=true
   )
   &CameraOffset(Camera Offset Amount=&FN_FrontEndFloatModifier(Gameplay=10, Front End=5))
  ]
  particleUpdateScript [
   &ParticleState()
   &SocketLocation(
    SamplePosOffset=true
    Sampled Position Offset=&Add_Vector(
     A=&RandomRangeVector(Minimum=<0 0 4>, Maximum=<0 0 -13>)
     B=&Multiply_Vector(
      A=&VectorFromFloat(Value=0.73)
      B=&MergeHelper_VectorSwitch(
       0=&VectorFromCurve(
        VectorCurve=DataInterface() {propertyValues {XCurve {keys [<at 0 0>,<at 1 1>]}, YCurve {keys [<at 0 0>,<at 1 1>]}, ZCurve {keys [<at 0 0>,<at 1 1>]}}}
        Scale Curve=&FN_FrontEndVector3Modifier(Gameplay=<25 13 -18>, Front End=<25 9 -18>)
       )
       1=&VectorFromCurve(
        VectorCurve=DataInterface() {propertyValues {XCurve {keys [<at 0 0>,<at 1 1>]}, YCurve {keys [<at 0 0>,<at 1 1>]}, ZCurve {keys [<at 0 0>,<at 1 1>]}}}
        Scale Curve=&FN_FrontEndVector3Modifier(Gameplay=<-25 13 -18>, Front End=<-25 9 -18>)
       )
       2=?
       3=?
       Index=&SpawnGroup()
       Num=::NewEnumerator0
      )
     )
    )
    Socket Index=&MergeHelper_IntSwitch(0=5, 1=6, 2=?, 3=?, 4=?, 5=?, Index=&SpawnGroup(), Num=::NewEnumerator0)
    Socket Reader=$System.Socket Reader
    Socket Sampling Mode=::NewEnumerator3
    VelocitySpeedLimit=false
   )
   &SpriteRotationRate(Rotation Rate=&RandomRangeFloat(Minimum=-50, Maximum=50))
   &ScaleSpriteSize(
    Scale Sprite Size Mode=::NewEnumerator0
    Uniform Scale Factor=&FloatFromCurve(
     FloatCurve=DataInterface() {propertyValues {Curve {
        defaultValue=1
        keys [
         {arriveTangent=4.1785, interpMode=RCIM_Cubic, leaveTangent=4.1785, tangentMode=RCTM_User, time=0, value=1}
         {arriveTangent=0.6516, interpMode=RCIM_Cubic, leaveTangent=0.6516, tangentMode=RCTM_User, time=0.2042, value=1.7491}
         {arriveTangent=0.0224, interpMode=RCIM_Cubic, leaveTangent=0.0224, tangentMode=RCTM_User, time=1, value=2}
        ]
       }}}
     Scale Curve=0.13
    )
   )
   &SubUVAnimation(
    End Frame Range Override=30
    "Sprite Renderer "=DataInterface() {propertyValues {SpriteRenderer=&self:SmokeDetailed_0.NiagaraSpriteRendererProperties_0}}
    Start Frame Range Override=&RandomRangeInt(Minimum=0, Maximum=5)
    UseEndFrame=true
    UseStartFrame=true
   )
   &ScaleColor(
    Scale Alpha=&Multiply_Float(
     A=&RandomRangeFloat(Minimum=1, Maximum=1)
     B=&FloatFromCurve(
      FloatCurve=DataInterface() {propertyValues {Curve {keys [
          {arriveTangent=3.7485, interpMode=RCIM_Cubic, leaveTangent=3.7485, tangentMode=RCTM_User, time=0, value=0.35}
          {interpMode=RCIM_Cubic, tangentMode=RCTM_User, time=0.5, value=1}
          {tangentMode=RCTM_User, time=1, value=0}
         ]}}}
      Scale Curve=&FN_FrontEndFloatModifier(Gameplay=2, Front End=1.5)
     )
    )
    Scale RGB=&MakeVectorFromLinearColorRGB(Color=&Multiply_LinearColorByFloat(
      Linear Color=&LerpLinearColors(StartColor=<rgba 0.016 0.016 0.016 1>, EndColor=<rgba 0.0573 0 0 1>, LerpFactor=$System.IncomingAudio)
      Float=&FNTimeOfDayScalar(Night Scale=0.6, Morning Scale=0.9, Evening Scale=0.8)
     ))
    ScaleA=true
    ScaleRGB=true
   )
  ]
  rendererClasses [&NiagaraSpriteRendererProperties]
  renderers [{rendererClass=&NiagaraSpriteRendererProperties}]
 }
}
summary {
 systemName=NS_Backpack_SummitReedMolt_Idle
 emitters [
  {emitterName=Smoke, rendererClasses [&NiagaraMeshRendererProperties]}
  {emitterName=RingsTop, rendererClasses [&NiagaraMeshRendererProperties, &NiagaraSpriteRendererProperties]}
  {emitterName=RingsMiddle, rendererClasses [&NiagaraMeshRendererProperties, &NiagaraSpriteRendererProperties]}
  {emitterName=SmokeDetailed, rendererClasses [&NiagaraSpriteRendererProperties]}
 ]
 userVariables [
  User.bIsFrontEnd {type=NiagaraBool}
  User.bIsFrontEndPreview {type=NiagaraBool}
  User.DebugReactivity {type=NiagaraBool}
  User.EnableFrontendFakeAudioLoop {type=NiagaraFloat, defaultValue=0, description=""}
  User.Reactivity {type=NiagaraFloat, defaultValue=0, description=""}
 ]
}
